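# Unlit, translucent, two-sided Niagara sprite material for loot star effects, tinted by particle color, which the asset name indicates carries rarity.
# Star mask and normal textures are sampled through 4x4 SubUV logic: a random-frame function by default, hand-built flipbook math on mobile.
# Branches add random or dynamic-parameter color and brightness, plus normal-based emissive modulation.
# A camera-distance offset collapses nearby sprites toward their centers, disabled on mobile.
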Material /DragonPopGameplayContent/Gameplay/FX/M_Loot_Star_RarityColor
graph {
 properties {
  bUsedWithNanite=true
  bUsedWithNiagaraSprites=true
  blendMode=BLEND_Translucent
  pixelDepthOffsetMode=PDOM_Legacy
  shadingModel=MSM_Unlit
  twoSided=true
  usedShadingModels=Unlit
 }
 nodes [
  n1 = Constant() {r=4}
  n2 = DynamicParameter() {paramNames [Manual Time Control, Lerp Color, Lerp Brightness, Param4]}
  n3 = ScalarParameter() {group=SubUV, parameterName=Manual Time Control, sortPriority=4}
  n4 = Constant() {r=4}
  n5 = Frac(None=Add(A=ParticleRandom()) {constB=0.05})
  n6 = Multiply(A=Time())
  n7 = StaticSwitch(
   True=MaterialFunctionCall!(
    UseTexCoord 1
    Rows=@n1
    Columns=@n1
    Particle Random Value=StaticSwitchParameter(True=@n2.Manual Time Control, False=@n3) {parameterName=Use Dynamic Parameter Control_1}
   ) {materialFunction=/Game/Effects/Fort_Effects/Materials/_Functions/MF_RandomSubUV_GPU_Sprites}.Result
   False=MaterialFunctionCall(Rows=@n4, Columns=@n4, Particle Random Value=Add(A=@n5, B=@n6)) {materialFunction=&MF_RandomSubUV_GPU_Sprites}.Result
   Value=StaticBoolParameter() {defaultValue=true, group=SubUV, parameterName=Use Manual Control, sortPriority=3}
  )
  n8 = TextureSampleParameter2D(UVs=@n7) {group=Texture, parameterName=Alpha Mask, samplerType=SAMPLERTYPE_Grayscale, texture=/Game/Effects/Fort_Effects/Textures/T_Stars_Mask}
  n9 = StaticComponentMaskParameter(None=@n8.RGB) {defaultG=true, group=Texture, parameterName=Alpha Component Mask, sortPriority=1}
  n10 = DynamicParameter() {paramNames [Manual Time Control, Lerp Color, Lerp Brightness, Param4]}
  n11 = ScalarParameter() {group=SubUV, parameterName=Manual Time Control, sortPriority=4}
  n12 = Multiply(
   A=ScalarParameter() {defaultValue=2, group=SubUV, parameterName=Timer Speed, sortPriority=2}
   B=StaticSwitch(
    True=StaticSwitchParameter(True=@n10.Manual Time Control, False=@n11) {group=SubUV, parameterName=Use Dynamic Parameter Control, sortPriority=5}
    False=Time() {period=1}
    Value=StaticBoolParameter() {defaultValue=true, group=SubUV, parameterName=Use Manual Control, sortPriority=3}
   )
  )
  n13 = ScalarParameter() {defaultValue=4, group=SubUV, parameterName=U Images}
  n14 = Floor(None=Multiply(A=Frac(None=@n12), B=@n13))
  n15 = Divide(B=@n13) {constB=1}
  n16 = TextureCoordinate()
  n17 = ComponentMask(None=@n16) {r=true}
  n18 = ComponentMask(None=@n16) {g=true}
  n19 = ScalarParameter() {defaultValue=4, group=SubUV, parameterName=V_Images, sortPriority=1}
  n20 = Divide(B=@n19) {constB=1}
  n21 = Divide(A=@n12, B=@n19)
  n22 = Floor(None=Multiply(A=Frac(None=@n21), B=@n19))
  n23 = AppendVector(A=Add(A=Divide(A=@n14, B=@n13), B=Multiply(A=@n15, B=@n17)), B=Add(A=Multiply(A=@n18, B=@n20), B=Divide(A=@n22, B=@n19)))
  n24 = TextureSampleParameter2D(UVs=@n23) {group=Texture, parameterName=Alpha Mask, samplerType=SAMPLERTYPE_Grayscale, texture=&T_Stars_Mask}
  n25 = StaticComponentMaskParameter(None=@n24.RGB) {defaultG=true, group=Texture, parameterName=Alpha Component Mask, sortPriority=1}
  FeatureLevelSwitch(Default=@n9, ES2=@n25, ES3_1=@n25, SM4_REMOVED=@n9)
  n26 = VectorParameter() {parameterName=ColorA, defaultValue=<rgba 3 1.5 0.5 1>}
  n27 = LinearInterpolate(
   A=@n26.RGB
   B=VectorParameter() {group=Random_Color, parameterName=ColorB, sortPriority=1, defaultValue=<rgba 0.5 0.25 0.0833 1>}.RGB
   Alpha=StaticSwitchParameter(True=DynamicParameter() {paramNames [Manual Time Control, Lerp Color, Lerp Brightness, Param4]}.Lerp Color, False=ParticleRandom()) {group=SubUV, parameterName=Use Dynamic Parameter Control, sortPriority=5}
  )
  FeatureLevelSwitch(Default=@n27, ES2=@n26.RGB, ES3_1=@n26.RGB, SM4_REMOVED=@n27)
  n28 = ParticlePositionWS()
  n29 = Subtract(A=Distance(A=@n28, B=CameraPositionWS()), B=ScalarParameter() {parameterName=DistanceOffset_1})
  n30 = ScalarParameter() {defaultValue=256, parameterName=ScaleUpDistance_1}
  n31 = LinearInterpolate(A=Subtract(A=@n28, B=WorldPosition().XYZ), Alpha=Clamp(None=Divide(A=@n29, B=@n30))) {constB=0}
  n32 = Constant()
  FeatureLevelSwitch(Default=@n31, ES2=@n32, ES3_1=@n32, SM4_REMOVED=@n31)
  n33 = ParticleColor()
  n34 = Multiply(A=@n33.RGB, B=@n8.RGB)
  n35 = Multiply(A=@n24.RGB, B=@n33.RGB)
  n36 = Multiply(B=FeatureLevelSwitch(Default=@n34, ES2=@n35, ES3_1=@n35, SM4_REMOVED=@n34))
  n37 = DynamicParameter() {paramNames [Manual Time Control, Lerp Color, Lerp Brightness, Param4]}
  n38 = ParticleRandom()
  n39 = Multiply(
   A=@n36
   B=LinearInterpolate(
    A=ScalarParameter() {defaultValue=1, group=Random_Color, parameterName=BrightLow, sortPriority=2}
    B=ScalarParameter() {defaultValue=8, group=Random_Color, parameterName=BrightHigh, sortPriority=3}
    Alpha=StaticSwitchParameter(True=@n37.Lerp Brightness, False=@n38) {group=SubUV, parameterName=Use Dynamic Parameter Control, sortPriority=5}
   )
  )
  n40 = FeatureLevelSwitch(Default=@n39, ES2=@n36, ES3_1=@n36, SM4_REMOVED=@n39)
  n41 = TextureSampleParameter2D(UVs=@n7) {group=Texture, parameterName=Normal, samplerType=SAMPLERTYPE_Normal, sortPriority=3, texture=/Game/Effects/Fort_Effects/Textures/T_Stars_NM}
  n42 = TextureSampleParameter2D(UVs=@n23) {group=Texture, parameterName=Normal, samplerType=SAMPLERTYPE_Normal, sortPriority=3, texture=&T_Stars_NM}
  n43 = DotProduct(A=FeatureLevelSwitch(Default=@n41.RGB, ES2=@n42.RGB, ES3_1=@n42.RGB, SM4_REMOVED=@n41.RGB), B=Constant3Vector() {constant=<rgba 1 0 0 0>})
  MaterialFunctionCall(Emissive=StaticSwitch(
    True=@n40
    False=Multiply(A=@n40, B=@n43)
    Value=StaticBoolParameter() {defaultValue=true, group=SubUV, parameterName=Use Manual Control, sortPriority=3}
   )) {materialFunction=&MF_FX_EmissiveAdjustments}
  Multiply(A=ParticleColor().RGB, B=@n34)
 ]
}
