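# M_RarityColor_Flare is an unlit translucent flare material that glows in an item rarity color, with depth test disabled.
# A Switch on Rarity_Index picks the color, BP_IsBlackListed blends it toward red, and a powered radial UV sphere mask shapes the emissive.
# Opacity runs that mask through depth fade, inverted Fresnel and OverallOpacity, gated by UseRarity.
# Camera-facing, camera-bias offset and scene-depth occlusion nodes exist but do not feed the material attributes.
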Material /DragonPopGameplayContent/Gameplay/FX/M_RarityColor_Flare
graph {
 expressions {
  MaterialExpressionAdd_0 {
   class=MaterialExpressionAdd
   inputs [
    A {from=MaterialExpressionWorldPosition_1, output=XYZ}
    B {from=MaterialExpressionSubtract_0}
   ]
  }
  MaterialExpressionAdd_1 {
   class=MaterialExpressionAdd
   inputs [
    A {from=MaterialExpressionAdd_0}
    B {from=MaterialExpressionReroute_0}
   ]
  }
  MaterialExpressionAdd_3 {
   class=MaterialExpressionAdd
   inputs [
    A {from=MaterialExpressionMultiply_4}
    B {from=MaterialExpressionMaterialFunctionCall_7, output=World Position Offset}
   ]
  }
  MaterialExpressionCameraPositionWS_1 {class=MaterialExpressionCameraPositionWS, inputs [], properties {}}
  MaterialExpressionConstant3Vector_1 {class=MaterialExpressionConstant3Vector, properties {constant=<rgba 0 0 -1 0>}}
  MaterialExpressionConstant3Vector_10 {class=MaterialExpressionConstant3Vector, properties {constant=<rgba 0 0.191 1 1>}}
  MaterialExpressionConstant3Vector_11 {class=MaterialExpressionConstant3Vector, properties {constant=<rgba 1 0.6431 0 1>}}
  MaterialExpressionConstant3Vector_12 {class=MaterialExpressionConstant3Vector, properties {constant=<rgba 0 0.6981 1 1>}}
  MaterialExpressionConstant3Vector_13 {class=MaterialExpressionConstant3Vector, properties {constant=<rgba 0.1516 1 0 1>}}
  MaterialExpressionConstant3Vector_14 {class=MaterialExpressionConstant3Vector, properties {constant=<rgba 1 0.2719 0 1>}}
  MaterialExpressionConstant3Vector_3 {class=MaterialExpressionConstant3Vector, properties {constant=<rgba 0 1 0 1>}}
  MaterialExpressionConstant3Vector_8 {class=MaterialExpressionConstant3Vector, properties {constant=<rgba 0.2475 0 1 1>}}
  MaterialExpressionConstant3Vector_9 {class=MaterialExpressionConstant3Vector, properties {constant=<rgba 0.1485 0.5573 0.0655 1>}}
  MaterialExpressionConstant_0 {class=MaterialExpressionConstant, properties {r=0.5}}
  MaterialExpressionConstant_8 {class=MaterialExpressionConstant, properties {r=0.5}}
  MaterialExpressionDepthFade_0 {
   class=MaterialExpressionDepthFade
   inputs [
    Opacity {from=MaterialExpressionPower_1, output=""}
    FadeDistance {from="", output=""}
   ]
   properties {fadeDistanceDefault=2}
  }
  MaterialExpressionLinearInterpolate_6 {
   class=MaterialExpressionLinearInterpolate
   inputs [
    A {from=MaterialExpressionSwitch_1}
    B {from=MaterialExpressionMultiply_28}
    Alpha {from=MaterialExpressionScalarParameter_16}
   ]
  }
  MaterialExpressionLinearInterpolate_9 {
   class=MaterialExpressionLinearInterpolate
   inputs [
    A {from=""}
    B {from=MaterialExpressionMultiply_2}
    Alpha {from=MaterialExpressionScalarParameter_17}
   ]
   properties {constB=0}
  }
  MaterialExpressionMaterialFunctionCall_0 {
   class=MaterialExpressionMaterialFunctionCall
   inputs [
    Normal Vector {from=""}
    Camera Vector {from=""}
    Invert Fresnel {from=MaterialExpressionStaticBool_0}
    Power {from=MaterialExpressionScalarParameter_1}
    Use Cheap contrast {from=""}
    Cheap contrast dark {from=""}
    Cheap contrast bright {from=""}
    Clamp Fresnel Dot Product {from=""}
   ]
   properties {materialFunction=&Fresnel_Function}
  }
  MaterialExpressionMaterialFunctionCall_1 {
   class=MaterialExpressionMaterialFunctionCall
   inputs [
    World Position {from=MaterialExpressionAdd_1}
    Buffer UV Pixel Offset {from=""}
   ]
   properties {materialFunction=&TransformToClipSpace}
  }
  MaterialExpressionMaterialFunctionCall_7 {
   class=MaterialExpressionMaterialFunctionCall
   inputs [
    WorldPosition {from=MaterialExpressionWorldPosition_2, output=XYZ}
    PivotLocation {from=""}
    WS Normals {from=""}
    Custom Object Basis 1 {from=""}
    Custom Object Basis 2 {from=""}
    Custom Object Basis 3 {from=""}
   ]
   properties {materialFunction=/Engine/Functions/Engine_MaterialFunctions02/WorldPositionOffset/AlignMeshToTheCamera}
  }
  MaterialExpressionMaterialFunctionCall_8 {
   class=MaterialExpressionMaterialFunctionCall
   inputs [
    Vector 3 {from=MaterialExpressionReroute_0}
    Vector 2 {from=""}
   ]
   properties {materialFunction=/Engine/Functions/Engine_MaterialFunctions02/Utility/VectorLength}
  }
  MaterialExpressionMultiply_0 {
   class=MaterialExpressionMultiply
   inputs [
    A {from=MaterialExpressionPower_1}
    B {from=MaterialExpressionLinearInterpolate_6}
   ]
  }
  MaterialExpressionMultiply_1 {
   class=MaterialExpressionMultiply
   inputs [
    A {from=MaterialExpressionDepthFade_0}
    B {from=MaterialExpressionMaterialFunctionCall_0, output=Result}
   ]
  }
  MaterialExpressionMultiply_17 {
   class=MaterialExpressionMultiply
   inputs [
    A {from=MaterialExpressionConstant3Vector_11}
    B {from=""}
   ]
   properties {constB=0.5}
  }
  MaterialExpressionMultiply_2 {
   class=MaterialExpressionMultiply
   inputs [
    A {from=MaterialExpressionMultiply_1}
    B {from=MaterialExpressionScalarParameter_3}
   ]
   properties {constB=0.25}
  }
  MaterialExpressionMultiply_20 {
   class=MaterialExpressionMultiply
   inputs [
    A {from=MaterialExpressionConstant3Vector_14}
    B {from=""}
   ]
  }
  MaterialExpressionMultiply_21 {
   class=MaterialExpressionMultiply
   inputs [
    A {from=MaterialExpressionConstant3Vector_10}
    B {from=""}
   ]
   properties {constB=0.5}
  }
  MaterialExpressionMultiply_24 {
   class=MaterialExpressionMultiply
   inputs [
    A {from=MaterialExpressionConstant3Vector_8}
    B {from=""}
   ]
   properties {constB=0.75}
  }
  MaterialExpressionMultiply_25 {
   class=MaterialExpressionMultiply
   inputs [
    A {from=MaterialExpressionConstant3Vector_12}
    B {from=""}
   ]
   properties {constB=0.75}
  }
  MaterialExpressionMultiply_28 {
   class=MaterialExpressionMultiply
   inputs [
    A {from=MaterialExpressionVectorParameter_1, output=RGB}
    B {from=""}
   ]
   properties {constB=2}
  }
  MaterialExpressionMultiply_3 {
   class=MaterialExpressionMultiply
   inputs [
    A {from=MaterialExpressionMultiply_0}
    B {from=""}
   ]
  }
  MaterialExpressionMultiply_4 {
   class=MaterialExpressionMultiply
   inputs [
    A {from=MaterialExpressionNormalize_0}
    B {from=MaterialExpressionScalarParameter_0}
   ]
  }
  MaterialExpressionNormalize_0 {class=MaterialExpressionNormalize, inputs [VectorInput {from=MaterialExpressionSubtract_3, output=""}]}
  MaterialExpressionObjectPositionWS_1 {class=MaterialExpressionObjectPositionWS, inputs [], properties {}}
  MaterialExpressionObjectPositionWS_2 {class=MaterialExpressionObjectPositionWS, inputs [], properties {}}
  MaterialExpressionPixelDepth_0 {class=MaterialExpressionPixelDepth, inputs [], properties {}}
  MaterialExpressionPower_1 {
   class=MaterialExpressionPower
   inputs [
    Base {from=MaterialExpressionSphereMask_1}
    Exp {from=MaterialExpressionScalarParameter_2}
   ]
   properties {constExponent=3}
  }
  MaterialExpressionReroute_0 {class=MaterialExpressionReroute, inputs [None {from=MaterialExpressionMultiply_4}]}
  MaterialExpressionSaturate_2 {class=MaterialExpressionSaturate, inputs [None {from=MaterialExpressionSubtract_1}]}
  MaterialExpressionScalarParameter_0 {class=MaterialExpressionScalarParameter, properties {defaultValue=25, parameterName=CameraBias}}
  MaterialExpressionScalarParameter_1 {class=MaterialExpressionScalarParameter, properties {defaultValue=2, parameterName=FresnelAmount}}
  MaterialExpressionScalarParameter_15 {class=MaterialExpressionScalarParameter, properties {defaultValue=2, group=HideoutsMarket, parameterName=Rarity_Index, sortPriority=0}}
  MaterialExpressionScalarParameter_16 {class=MaterialExpressionScalarParameter, properties {group=HideoutsMarket, parameterName=BP_IsBlackListed}}
  MaterialExpressionScalarParameter_17 {class=MaterialExpressionScalarParameter, properties {defaultValue=1, parameterName=UseRarity}}
  MaterialExpressionScalarParameter_2 {class=MaterialExpressionScalarParameter, properties {defaultValue=4, parameterName=FlareExp}}
  MaterialExpressionScalarParameter_3 {class=MaterialExpressionScalarParameter, properties {defaultValue=0.2, parameterName=OverallOpacity}}
  MaterialExpressionScalarParameter_4 {class=MaterialExpressionScalarParameter, properties {defaultValue=0.1, parameterName=FlareHardness}}
  MaterialExpressionSceneDepth_0 {
   class=MaterialExpressionSceneDepth
   inputs [UVs {from=MaterialExpressionMaterialFunctionCall_1, output=Screen aligned 0-1 UV}]
   properties {}
  }
  MaterialExpressionSetMaterialAttributes_0 {
   class=MaterialExpressionSetMaterialAttributes
   inputs [
    MaterialAttributes {from="", output=""}
    Emissive Color {from=MaterialExpressionMultiply_3, output=""}
    World Position Offset {from="", output=""}
    Opacity {from=MaterialExpressionLinearInterpolate_9, output=""}
   ]
   properties {attributeSetTypes [#1, #2, #3]}
  }
  MaterialExpressionSphereMask_1 {
   class=MaterialExpressionSphereMask
   inputs [
    A {from=MaterialExpressionTextureCoordinate_0, output=""}
    B {from=MaterialExpressionConstant_0, output=""}
    Radius {from=MaterialExpressionConstant_0, output=""}
    Hardness {from=MaterialExpressionScalarParameter_4, output=""}
   ]
   properties {}
  }
  MaterialExpressionStaticBool_0 {class=MaterialExpressionStaticBool, properties {value=true}}
  MaterialExpressionSubtract_0 {
   class=MaterialExpressionSubtract
   inputs [
    A {from=MaterialExpressionObjectPositionWS_1}
    B {from=MaterialExpressionWorldPosition_0, output=XYZ}
   ]
  }
  MaterialExpressionSubtract_1 {
   class=MaterialExpressionSubtract
   inputs [
    A {from=MaterialExpressionSceneDepth_0, output=R}
    B {from=MaterialExpressionSubtract_4}
   ]
  }
  MaterialExpressionSubtract_3 {
   class=MaterialExpressionSubtract
   inputs [
    A {from=MaterialExpressionCameraPositionWS_1}
    B {from=MaterialExpressionObjectPositionWS_2}
   ]
  }
  MaterialExpressionSubtract_4 {
   class=MaterialExpressionSubtract
   inputs [
    A {from=MaterialExpressionPixelDepth_0, output=R}
    B {from=MaterialExpressionMaterialFunctionCall_8, output=V3 Length}
   ]
  }
  MaterialExpressionSwitch_1 {
   class=MaterialExpressionSwitch
   inputs [
    SwitchValue {from=MaterialExpressionScalarParameter_15, output=""}
    Default {from="", output=""}
    None {from=MaterialExpressionConstant_8, output=""}
    None {from=MaterialExpressionConstant3Vector_9, output=""}
    None {from=MaterialExpressionMultiply_21, output=""}
    None {from=MaterialExpressionMultiply_24, output=""}
    None {from=MaterialExpressionMultiply_17, output=""}
    None {from=MaterialExpressionMultiply_20, output=""}
    None {from=MaterialExpressionMultiply_25, output=""}
    None {from=MaterialExpressionConstant3Vector_13, output=""}
    None {from=MaterialExpressionConstant3Vector_13, output=""}
    None {from=MaterialExpressionConstant3Vector_13, output=""}
   ]
   properties {inputs [
     {
      inputName=None
      input {
       expression=&self:MaterialExpressionConstant_8
       inputName=None
       mask=0
       maskA=0
       maskB=0
       maskG=0
       maskR=0
       outputIndex=0
      }
     }
     {
      inputName=None
      input {
       expression=&self:MaterialExpressionConstant3Vector_9
       inputName=None
       mask=1
       maskA=0
       maskB=1
       maskG=1
       maskR=1
       outputIndex=0
      }
     }
     {
      inputName=None
      input {
       expression=&self:MaterialExpressionMultiply_21
       inputName=None
       mask=0
       maskA=0
       maskB=0
       maskG=0
       maskR=0
       outputIndex=0
      }
     }
     {
      inputName=None
      input {
       expression=&self:MaterialExpressionMultiply_24
       inputName=None
       mask=0
       maskA=0
       maskB=0
       maskG=0
       maskR=0
       outputIndex=0
      }
     }
     {
      inputName=None
      input {
       expression=&self:MaterialExpressionMultiply_17
       inputName=None
       mask=0
       maskA=0
       maskB=0
       maskG=0
       maskR=0
       outputIndex=0
      }
     }
     {
      inputName=None
      input {
       expression=&self:MaterialExpressionMultiply_20
       inputName=None
       mask=0
       maskA=0
       maskB=0
       maskG=0
       maskR=0
       outputIndex=0
      }
     }
     {
      inputName=None
      input {
       expression=&self:MaterialExpressionMultiply_25
       inputName=None
       mask=0
       maskA=0
       maskB=0
       maskG=0
       maskR=0
       outputIndex=0
      }
     }
     {
      inputName=None
      input {
       expression=&self:MaterialExpressionConstant3Vector_13
       inputName=None
       mask=1
       maskA=0
       maskB=1
       maskG=1
       maskR=1
       outputIndex=0
      }
     }
     {
      inputName=None
      input {
       expression=&self:MaterialExpressionConstant3Vector_13
       inputName=None
       mask=1
       maskA=0
       maskB=1
       maskG=1
       maskR=1
       outputIndex=0
      }
     }
     {
      inputName=None
      input {
       expression=&self:MaterialExpressionConstant3Vector_13
       inputName=None
       mask=1
       maskA=0
       maskB=1
       maskG=1
       maskR=1
       outputIndex=0
      }
     }
    ]}
  }
  MaterialExpressionTextureCoordinate_0 {class=MaterialExpressionTextureCoordinate}
  MaterialExpressionVectorParameter_1 {class=MaterialExpressionVectorParameter, properties {parameterName=BlackListed_RedColor, defaultValue=<rgba 1 0 0.0035 1>}}
  MaterialExpressionWorldPosition_0 {class=MaterialExpressionWorldPosition, inputs [], properties {}}
  MaterialExpressionWorldPosition_1 {class=MaterialExpressionWorldPosition, inputs [], properties {}}
  MaterialExpressionWorldPosition_2 {class=MaterialExpressionWorldPosition, inputs [], properties {}}
 }
 properties {
  bDisableDepthTest=true
  bUseMaterialAttributes=true
  bUseTranslucencyVertexFog=false
  blendMode=BLEND_Translucent
  pixelDepthOffsetMode=PDOM_Legacy
  shadingModel=MSM_Unlit
  usedShadingModels=Unlit
 }
}
