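# Hideouts market material that computes a glow color from an item's rarity, with a red override for blacklisted entries.
# A Switch on Rarity_Index picks a preset color, which GlowIntensity and UseRarity scale.
# BP_IsBlackListed blends this toward the red parameter, and dividing by EyeAdaptation counters auto exposure.
# A second blend yields the unadapted rarity color, or a near-black constant when rarity is off or the entry is blacklisted.
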Material /DragonPopGameplayContent/Gameplay/Materials/M_HideoutsMarket_RarityEmissive
graph {
 expressions {
  MaterialExpressionConstant3Vector_1 {class=MaterialExpressionConstant3Vector, properties {constant=<rgba 0.1118 0 1 1>}}
  MaterialExpressionConstant3Vector_2 {class=MaterialExpressionConstant3Vector, properties {constant=<rgba 0.1485 0.5573 0.0655 1>}}
  MaterialExpressionConstant3Vector_3 {class=MaterialExpressionConstant3Vector, properties {constant=<rgba 0.003 0.0186 1 1>}}
  MaterialExpressionConstant3Vector_4 {class=MaterialExpressionConstant3Vector, properties {constant=<rgba 1 0.5916 0.0319 1>}}
  MaterialExpressionConstant3Vector_5 {class=MaterialExpressionConstant3Vector, properties {constant=<rgba 0 0.7541 1 1>}}
  MaterialExpressionConstant3Vector_6 {class=MaterialExpressionConstant3Vector, properties {constant=<rgba 0.1516 1 0 1>}}
  MaterialExpressionConstant3Vector_7 {class=MaterialExpressionConstant3Vector, properties {constant=<rgba 1 0.1906 0 1>}}
  MaterialExpressionConstant_0 {class=MaterialExpressionConstant, properties {r=0.02}}
  MaterialExpressionConstant_6 {class=MaterialExpressionConstant, properties {r=0.5}}
  MaterialExpressionDivide_0 {
   class=MaterialExpressionDivide
   inputs [
    A {from=MaterialExpressionLinearInterpolate_0}
    B {from=MaterialExpressionEyeAdaptation_0, output=EyeAdaptation}
   ]
  }
  MaterialExpressionEyeAdaptation_0 {class=MaterialExpressionEyeAdaptation, inputs [], properties {}}
  MaterialExpressionLinearInterpolate_0 {
   class=MaterialExpressionLinearInterpolate
   inputs [
    A {from=MaterialExpressionMultiply_28}
    B {from=MaterialExpressionMultiply_3}
    Alpha {from=MaterialExpressionScalarParameter_1}
   ]
  }
  MaterialExpressionLinearInterpolate_1 {
   class=MaterialExpressionLinearInterpolate
   inputs [
    A {from=MaterialExpressionSwitch_1}
    B {from=MaterialExpressionConstant_0}
    Alpha {from=MaterialExpressionScalarParameter_1}
   ]
  }
  MaterialExpressionMultiply_0 {
   class=MaterialExpressionMultiply
   inputs [
    A {from=MaterialExpressionConstant3Vector_3}
    B {from=""}
   ]
   properties {constB=0.3}
  }
  MaterialExpressionMultiply_1 {
   class=MaterialExpressionMultiply
   inputs [
    A {from=MaterialExpressionConstant3Vector_1}
    B {from=""}
   ]
   properties {constB=0.5}
  }
  MaterialExpressionMultiply_16 {
   class=MaterialExpressionMultiply
   inputs [
    A {from=MaterialExpressionSwitch_0}
    B {from=MaterialExpressionScalarParameter_0}
   ]
   properties {constB=5}
  }
  MaterialExpressionMultiply_2 {
   class=MaterialExpressionMultiply
   inputs [
    A {from=MaterialExpressionConstant3Vector_5}
    B {from=""}
   ]
   properties {constB=1.5}
  }
  MaterialExpressionMultiply_26 {
   class=MaterialExpressionMultiply
   inputs [
    A {from=MaterialExpressionConstant3Vector_4}
    B {from=""}
   ]
   properties {constB=1.5}
  }
  MaterialExpressionMultiply_27 {
   class=MaterialExpressionMultiply
   inputs [
    A {from=MaterialExpressionConstant3Vector_7}
    B {from=""}
   ]
  }
  MaterialExpressionMultiply_28 {
   class=MaterialExpressionMultiply
   inputs [
    A {from=MaterialExpressionScalarParameter_32}
    B {from=MaterialExpressionMultiply_16}
   ]
  }
  MaterialExpressionMultiply_3 {
   class=MaterialExpressionMultiply
   inputs [
    A {from=MaterialExpressionVectorParameter_0, output=RGB}
    B {from=""}
   ]
   properties {constB=2}
  }
  MaterialExpressionScalarParameter_0 {class=MaterialExpressionScalarParameter, properties {defaultValue=3, group=HideoutsMarket, parameterName=GlowIntensity, sortPriority=1}}
  MaterialExpressionScalarParameter_1 {class=MaterialExpressionScalarParameter, properties {group=HideoutsMarket, parameterName=BP_IsBlackListed}}
  MaterialExpressionScalarParameter_30 {class=MaterialExpressionScalarParameter, properties {defaultValue=5, group=HideoutsMarket, parameterName=Rarity_Index, sortPriority=0}}
  MaterialExpressionScalarParameter_32 {class=MaterialExpressionScalarParameter, properties {defaultValue=1, group=HideoutsMarket, parameterName=UseRarity, sortPriority=0}}
  MaterialExpressionSwitch_0 {
   class=MaterialExpressionSwitch
   inputs [
    SwitchValue {from=MaterialExpressionScalarParameter_30, output=""}
    Default {from="", output=""}
    None {from=MaterialExpressionConstant_6, output=""}
    None {from=MaterialExpressionConstant3Vector_2, output=""}
    None {from=MaterialExpressionMultiply_0, output=""}
    None {from=MaterialExpressionMultiply_1, output=""}
    None {from=MaterialExpressionMultiply_26, output=""}
    None {from=MaterialExpressionMultiply_27, output=""}
    None {from=MaterialExpressionMultiply_2, output=""}
    None {from=MaterialExpressionConstant3Vector_6, output=""}
    None {from=MaterialExpressionConstant3Vector_6, output=""}
    None {from=MaterialExpressionConstant3Vector_6, output=""}
   ]
   properties {inputs [
     {
      inputName=None
      input {
       expression=&self:MaterialExpressionConstant_6
       inputName=None
       mask=0
       maskA=0
       maskB=0
       maskG=0
       maskR=0
       outputIndex=0
      }
     }
     {
      inputName=None
      input {
       expression=&self:MaterialExpressionConstant3Vector_2
       inputName=None
       mask=1
       maskA=0
       maskB=1
       maskG=1
       maskR=1
       outputIndex=0
      }
     }
     {
      inputName=None
      input {
       expression=&self:MaterialExpressionMultiply_0
       inputName=None
       mask=0
       maskA=0
       maskB=0
       maskG=0
       maskR=0
       outputIndex=0
      }
     }
     {
      inputName=None
      input {
       expression=&self:MaterialExpressionMultiply_1
       inputName=None
       mask=0
       maskA=0
       maskB=0
       maskG=0
       maskR=0
       outputIndex=0
      }
     }
     {
      inputName=None
      input {
       expression=&self:MaterialExpressionMultiply_26
       inputName=None
       mask=0
       maskA=0
       maskB=0
       maskG=0
       maskR=0
       outputIndex=0
      }
     }
     {
      inputName=None
      input {
       expression=&self:MaterialExpressionMultiply_27
       inputName=None
       mask=0
       maskA=0
       maskB=0
       maskG=0
       maskR=0
       outputIndex=0
      }
     }
     {
      inputName=None
      input {
       expression=&self:MaterialExpressionMultiply_2
       inputName=None
       mask=0
       maskA=0
       maskB=0
       maskG=0
       maskR=0
       outputIndex=0
      }
     }
     {
      inputName=None
      input {
       expression=&self:MaterialExpressionConstant3Vector_6
       inputName=None
       mask=1
       maskA=0
       maskB=1
       maskG=1
       maskR=1
       outputIndex=0
      }
     }
     {
      inputName=None
      input {
       expression=&self:MaterialExpressionConstant3Vector_6
       inputName=None
       mask=1
       maskA=0
       maskB=1
       maskG=1
       maskR=1
       outputIndex=0
      }
     }
     {
      inputName=None
      input {
       expression=&self:MaterialExpressionConstant3Vector_6
       inputName=None
       mask=1
       maskA=0
       maskB=1
       maskG=1
       maskR=1
       outputIndex=0
      }
     }
    ]}
  }
  MaterialExpressionSwitch_1 {
   class=MaterialExpressionSwitch
   inputs [
    SwitchValue {from=MaterialExpressionScalarParameter_32, output=""}
    Default {from=MaterialExpressionMultiply_28, output=""}
    None {from=MaterialExpressionConstant_0, output=""}
    None {from=MaterialExpressionMultiply_28, output=""}
   ]
   properties {inputs [
     {
      inputName=None
      input {
       expression=&self:MaterialExpressionConstant_0
       inputName=None
       mask=0
       maskA=0
       maskB=0
       maskG=0
       maskR=0
       outputIndex=0
      }
     }
     {
      inputName=None
      input {
       expression=&self:MaterialExpressionMultiply_28
       inputName=None
       mask=0
       maskA=0
       maskB=0
       maskG=0
       maskR=0
       outputIndex=0
      }
     }
    ]}
  }
  MaterialExpressionVectorParameter_0 {class=MaterialExpressionVectorParameter, properties {parameterName=BlackListed_RedColor, defaultValue=<rgba 1 0 0.0035 1>}}
 }
 properties {bAutomaticallySetUsageInEditor=false, usedShadingModels=Default Lit}
}
